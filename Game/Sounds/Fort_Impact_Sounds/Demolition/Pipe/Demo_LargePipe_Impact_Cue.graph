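# Sound cue for large pipe impacts in demolition, routed to the impacts sound class.
# A quality-level switch selects among four branches: three randomly pick between two impact waves with slight pitch modulation, while the last plays one wave with wider pitch variation.
# Small-object demolition attenuation handles spatial falloff, and a four-max object-impact concurrency set limits simultaneous hits.
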SoundCue /Game/Sounds/Fort_Impact_Sounds/Demolition/Pipe/Demo_LargePipe_Impact_Cue
root=QualityLevel_0
nodes [
 {id=Modulator_0, children [Random_0], properties {pitchMax=1.07, pitchMin=0.92}}
 {id=Modulator_1, children [Random_2], properties {pitchMax=1.07, pitchMin=0.92}}
 {id=Modulator_2, children [Random_1], properties {pitchMax=1.07, pitchMin=0.92}}
 {id=Modulator_3, children [WavePlayer_6], properties {pitchMax=1.2, pitchMin=0.9}}
 {id=QualityLevel_0, children [Modulator_2,Modulator_0,Modulator_1,Modulator_3]}
 {id=Random_0, children [WavePlayer_0,WavePlayer_1], properties {weights [1,1]}}
 {id=Random_1, children [WavePlayer_4,WavePlayer_5], properties {weights [1,1]}}
 {id=Random_2, children [WavePlayer_2,WavePlayer_3], properties {weights [1,1]}}
 {id=WavePlayer_0, properties {soundWaveAssetPtr=&/pipe_large_impact_02}}
 {id=WavePlayer_1, properties {soundWaveAssetPtr=&/pipe_large_impact_03}}
 {id=WavePlayer_2, properties {soundWaveAssetPtr=&pipe_large_impact_02}}
 {id=WavePlayer_3, properties {soundWaveAssetPtr=&pipe_large_impact_03}}
 {id=WavePlayer_4, properties {soundWaveAssetPtr=&pipe_large_impact_02}}
 {id=WavePlayer_5, properties {soundWaveAssetPtr=&pipe_large_impact_03}}
 {id=WavePlayer_6, properties {soundWaveAssetPtr=&pipe_large_impact_02}}
]
properties {
 attenuationSettings=/Game/Sounds/Attenuation/Fort_Demolition_Hit_Small_Sized_Object_Att
 maxDistance=4000
 soundClassObject=&Fort_Impacts
 volumeMultiplier=0.68
 concurrencyOverrides {maxCount=4}
 concurrencySet [&CONC_Object_Impacts_FourMax]
}
