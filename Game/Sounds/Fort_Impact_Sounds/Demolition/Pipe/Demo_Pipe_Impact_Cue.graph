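# Sound cue for medium pipe impacts in the demolition impact sound set.
# A random node picks one of three equally weighted pipe impact waves, and a modulator varies its pitch slightly on each play.
# It plays in the Fort_Impacts sound class with small-object demolition hit attenuation, a reduced volume, and a concurrency set that caps simultaneous object impacts at four.
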SoundCue /Game/Sounds/Fort_Impact_Sounds/Demolition/Pipe/Demo_Pipe_Impact_Cue
root=Modulator_14
nodes [
 {id=Modulator_14, children [Random_0], properties {pitchMax=1.07, pitchMin=0.92}}
 {id=Random_0, children [WavePlayer_0,WavePlayer_1,WavePlayer_2], properties {weights [1,1,1]}}
 {id=WavePlayer_0, properties {soundWaveAssetPtr=&/pipe_med_impact_01}}
 {id=WavePlayer_1, properties {soundWaveAssetPtr=&/pipe_med_impact_02}}
 {id=WavePlayer_2, properties {soundWaveAssetPtr=&/pipe_med_impact_03}}
]
properties {
 attenuationSettings=/Game/Sounds/Attenuation/Fort_Demolition_Hit_Small_Sized_Object_Att
 maxDistance=4000
 soundClassObject=&Fort_Impacts
 volumeMultiplier=0.68
 concurrencyOverrides {maxCount=4}
 concurrencySet [&CONC_Object_Impacts_FourMax]
}
